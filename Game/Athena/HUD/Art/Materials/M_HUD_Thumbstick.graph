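# Translucent UI material that procedurally draws a tintable HUD thumbstick graphic from UV distance to the center.
# A static switch selects either a smoothstepped gradient circle or a composite of border and middle shapes, shaped with SmoothCeil and merged by lighten blending.
# The composite can also lighten-blend in a separately scaled lock icon's alpha through the Lock switch.
# The Color parameter multiplies the resulting mask for color and clamped alpha.
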Material /Game/Athena/HUD/Art/Materials/M_HUD_Thumbstick
graph {
 properties {blendMode=BLEND_Translucent, materialDomain=MD_UI, pixelDepthOffsetMode=PDOM_Legacy, usedShadingModels=Default Lit}
 nodes [
  n1 = TextureCoordinate() {coordinateIndex=1}
  n2 = ScalarParameter() {defaultValue=1, parameterName=Main scale}
  n3 = Distance(
   A=Add(A=MaterialFunctionCall(UVs=@n1, Texture Scale=@n2) {materialFunction=&ScaleUVsByCenter}.UVs) {constB=-0.5}
   B=Constant()
  )
  n4 = MaterialFunctionCall(
   Alpha=OneMinus(None=@n3)
   Min=ScalarParameter() {defaultValue=0.56, parameterName=Smoothstep Min}
   Max=ScalarParameter() {defaultValue=1, parameterName=Smoothstep Max}
  ) {materialFunction=&SmoothStep}
  n5 = MaterialFunctionCall(UVs=TextureCoordinate(), Texture Scale=ScalarParameter() {defaultValue=1, parameterName=Lock scale}) {materialFunction=&ScaleUVsByCenter}
  n6 = Sine(None=@n3)
  n7 = ScalarParameter() {defaultValue=0.05, parameterName=Border Width}
  n8 = MaterialFunctionCall!(In=ScalarParameter() {defaultValue=-0.1, parameterName=Smooth border, sliderMin=-0.3}, Ramp Width=Clamp(None=Power(Base=@n6, Exp=@n7))) {materialFunction=/Engine/Functions/Engine_MaterialFunctions01/Math/SmoothCeil}
  n9 = ScalarParameter() {defaultValue=0.9, parameterName=Border opacity}
  n10 = MaterialFunctionCall(UVs=@n1, Texture Scale=Multiply(A=@n2, B=ScalarParameter() {defaultValue=0.4, parameterName=Middle scale}) {constB=0.4}) {materialFunction=&ScaleUVsByCenter}
  n11 = Clamp(
   None=Distance(A=Add(A=@n10.UVs) {constB=-0.5}, B=Constant())
   Min=ScalarParameter() {defaultValue=0.4, parameterName=Middle Min}
   Max=ScalarParameter() {defaultValue=0.9, parameterName=Middle Max}
  ) {minDefault=0.3429}
  n12 = MaterialFunctionCall(In=ScalarParameter() {defaultValue=0.181, parameterName=Middle Border}, Ramp Width=@n11) {materialFunction=&SmoothCeil}
  n13 = Multiply(A=@n11, B=OneMinus(None=Round(None=@n12.Result)))
  n14 = ScalarParameter() {defaultValue=1, parameterName=Middle Opacity}
  n15 = MaterialFunctionCall!(
   Blend=StaticSwitchParameter(
    True=TextureSampleParameter2D(UVs=@n5.UVs) {parameterName=Icon, texture=/Game/UI/Foundation/Textures/Icons/Locks/T-Icon-Lock-32}.A
    False=Constant()
   ) {parameterName=Lock}
   Base=MaterialFunctionCall(Blend=Clamp(None=@n8.Result, Max=@n9), Base=Power(Base=@n13, Exp=@n14)) {materialFunction=/Engine/Functions/Engine_MaterialFunctions03/Blends/Blend_Lighten}.Result
  ) {materialFunction=&Blend_Lighten}
  n16 = VectorParameter() {parameterName=Color, defaultValue=<rgba 1 1 1 1>}
  Clamp(None=Multiply(A=StaticSwitchParameter(True=@n4.Result, False=@n15.Result) {parameterName=Use Gradient Circle?}, B=@n16.A))
  Multiply(A=StaticSwitchParameter(True=@n4.Result, False=@n15.Result) {parameterName=Use Gradient Circle?}, B=@n16.RGB)
 ]
}
